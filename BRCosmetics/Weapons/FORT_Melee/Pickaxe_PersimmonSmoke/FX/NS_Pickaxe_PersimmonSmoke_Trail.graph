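# Niagara ribbon trail effect for the PersimmonSmoke pickaxe, drawing a grey blade streak with a fiery orange glow along the weapon's animation trail.
# Three emitters use the FN AnimTrail spawn, initialize and update modules: two short-lived blade ribbons and a bright fire trail with two ribbon renderers.
# Curves fade each trail over particle life, system-level TOD and TODFire colors tint them, and the fire trail computes a separate GlowTrailColor.
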NiagaraSystem /BRCosmetics/Weapons/FORT_Melee/Pickaxe_PersimmonSmoke/FX/NS_Pickaxe_PersimmonSmoke_Trail
dependencies {
 usedDataInterfaces [&NiagaraDataInterfaceCurve, &NiagaraDataInterfaceSocketReader, &NiagaraDataInterfaceColorCurve]
 usedDynamicInputs [&ColorFromCurve]
 usedModules [
  &self:SystemSpawnScript.SystemScriptSource.SystemScriptGraph.NiagaraNodeAssignment_0.SetVariables_#1
  &self:SystemSpawnScript.SystemScriptSource.SystemScriptGraph.NiagaraNodeAssignment_2.SetVariables_#2
  &self:SystemSpawnScript.SystemScriptSource.SystemScriptGraph.NiagaraNodeAssignment_1.SetVariables_#3
  &SystemState
 ]
}
emitters {
 TrailBlade {
  emitterUpdateScript [
   &EmitterState()
   &FN_SpawnAnimTrail(AnimTrailWidthMode=::FromSecond)
  ]
  particleSpawnScript [
   &FN_InitializeAnimTrailBoneIndecies() {module=FN_InitializeAnimTrailBoneIndecies001}
   &FN_InitializeAnimTrail(Color=<rgba 0.1425 0.141 0.15 1>, Lifetime=0.15) {module=FN_InitializeAnimTrail001}
  ]
  particleUpdateScript [
   &ScaleColor(
    Scale Alpha=&Multiply_Float(
     A=&FloatFromCurve(FloatCurve=DataInterface() {propertyValues {Curve {keys [
          {interpMode=RCIM_Cubic, leaveTangent=-1.8709, tangentMode=RCTM_Break, time=0, value=1}
          {arriveTangent=-0.3327, interpMode=RCIM_Cubic, leaveTangent=-3.6795, tangentMode=RCTM_Break, time=1, value=0}
         ]}}})
     B=2
    )
    Scale RGB=&MakeVectorFromLinearColorRGB(Color=$System.TOD)
    ScaleA=true
    ScaleRGB=true
   ) {module=ScaleColor002}
   &FN_UpdateAnimTrail()
   &SolveForcesAndVelocity(Clamp Velocity=false, Limit Acceleration=false)
  ]
  rendererClasses [&NiagaraRibbonRendererProperties]
  renderers [{rendererClass=&NiagaraRibbonRendererProperties}]
 }
 TrailBlade001 {
  emitterUpdateScript [
   &EmitterState()
   &FN_SpawnAnimTrail(AnimTrailWidthMode=::FromSecond)
  ]
  particleSpawnScript [
   &FN_InitializeAnimTrailBoneIndecies() {module=FN_InitializeAnimTrailBoneIndecies001}
   &FN_InitializeAnimTrail(AnimTrailWidth=0.8, Color=<rgba 0.1412 0.1395 0.15 1>, Lifetime=0.12) {module=FN_InitializeAnimTrail001}
  ]
  particleUpdateScript [
   &ScaleColor(
    Scale Alpha=&Multiply_Float(
     A=&FloatFromCurve(FloatCurve=DataInterface() {propertyValues {Curve {keys [
          {interpMode=RCIM_Cubic, leaveTangent=-1.8709, tangentMode=RCTM_Break, time=0, value=1}
          {arriveTangent=-0.3327, interpMode=RCIM_Cubic, leaveTangent=-3.6795, tangentMode=RCTM_Break, time=1, value=0}
         ]}}})
     B=2
    )
    Scale RGB=&MakeVectorFromLinearColorRGB(Color=$System.TOD)
    ScaleA=true
    ScaleRGB=true
   ) {module=ScaleColor002}
   &FN_UpdateAnimTrail()
   &SolveForcesAndVelocity(Clamp Velocity=false, Limit Acceleration=false)
  ]
  rendererClasses [&NiagaraRibbonRendererProperties]
  renderers [{rendererClass=&NiagaraRibbonRendererProperties}]
 }
 TrailFire {
  emitterUpdateScript [
   &EmitterState()
   &FN_SpawnAnimTrail()
  ]
  particleSpawnScript [
   &FN_InitializeAnimTrailBoneIndecies() {module=FN_InitializeAnimTrailBoneIndecies001}
   &FN_InitializeAnimTrail(AnimTrailWidth=1.2, Color=<rgba 10 2.2842 0.15 0.5>, Lifetime=0.2) {module=FN_InitializeAnimTrail001}
  ]
  particleUpdateScript [
   &ScaleColor(
    Scale Alpha=&Multiply_Float(
     A=&FloatFromCurve(FloatCurve=DataInterface() {propertyValues {Curve {keys [
          {interpMode=RCIM_Cubic, leaveTangent=-16.9095, tangentMode=RCTM_Break, time=0, value=1}
          {arriveTangent=-0.3327, interpMode=RCIM_Cubic, leaveTangent=-3.6795, tangentMode=RCTM_Break, time=0.2, value=0}
         ]}}})
     B=1
    )
    Scale RGB=&Multiply_VectorByFloat(
     Vector=&MakeVectorFromLinearColorRGB(Color=$System.TODFire)
     Float=&FloatFromCurve(FloatCurve=DataInterface() {propertyValues {Curve {keys [
          {interpMode=RCIM_Cubic, leaveTangent=-5.8489, tangentMode=RCTM_Break, time=0, value=1}
          {arriveTangent=-0.3327, interpMode=RCIM_Cubic, leaveTangent=-3.6795, tangentMode=RCTM_Break, time=0.2, value=0}
         ]}}})
    )
    ScaleA=true
    ScaleRGB=true
   ) {module=ScaleColor002}
   &FN_UpdateAnimTrail()
   /SetVariables(Particles.GlowTrailColor:LinearColor=&MakeLinearColorFromVectorAndFloat(
     "Vector (RGB)"=&Multiply_Vector(
      A=&Multiply_VectorByFloat(
       Vector=&MakeVectorFromLinearColorRGB(Color=<rgba 0.5 0.0719 0.006 1>)
       Float=&FloatFromCurve(FloatCurve=DataInterface() {propertyValues {Curve {keys [{interpMode=RCIM_Cubic, tangentMode=RCTM_User, time=0, value=1}, {interpMode=RCIM_Cubic, tangentMode=RCTM_User, time=1, value=0}]}}})
      )
      B=&MakeVectorFromLinearColorRGB(Color=$System.TODFire)
     )
     "Float (Alpha)"=&Multiply_Float(
      A=&FloatFromCurve(FloatCurve=DataInterface() {propertyValues {Curve {keys [
           {interpMode=RCIM_Cubic, leaveTangent=-3.2813, tangentMode=RCTM_Break, time=0, value=1}
           {arriveTangent=-0.0411, interpMode=RCIM_Cubic, tangentMode=RCTM_Break, time=0.5, value=0}
          ]}}})
      B=1
     )
    )) {isSetParameters=true}
   &ScaleRibbonWidth(Ribbon Width Scale=&FloatFromCurve(FloatCurve=DataInterface() {propertyValues {Curve {keys [{arriveTangent=-2.4001, interpMode=RCIM_Cubic, leaveTangent=-2.4001, tangentMode=RCTM_User, time=0, value=1}, <at 1 0 RCIM_Cubic>]}}})) {enabled=false}
   &SolveForcesAndVelocity(Clamp Velocity=false, Limit Acceleration=false)
  ]
  rendererClasses [&NiagaraRibbonRendererProperties]
  renderers [
   {rendererClass=&NiagaraRibbonRendererProperties}
   {rendererClass=&NiagaraRibbonRendererProperties, rendererIndex=1}
  ]
 }
}
summary {
 systemName=NS_Pickaxe_PersimmonSmoke_Trail
 emitters [
  {emitterName=TrailFire, rendererClasses [&NiagaraRibbonRendererProperties]}
  {emitterName=TrailBlade, rendererClasses [&NiagaraRibbonRendererProperties]}
  {emitterName=TrailBlade001, rendererClasses [&NiagaraRibbonRendererProperties]}
 ]
 userVariables [
  User.bIsFrontEnd {type=NiagaraBool}
  User.bIsFrontEndPreview {type=NiagaraBool}
  User.RibbonIDOffset {type=NiagaraInt32}
 ]
}
